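# Niagara reference system for the Flowing Flames emote, combining stylised water trails with fire effects on a character's skeletal mesh.
# Toon_Water and Glow_and_Liquid draw animation-trail ribbons faded by a shared swing value, alongside socket-sampled droplets and a burst of splash sprites.
# Twelve fire meshes follow selected bones through partition-indexed arrays, while two-group flame/smoke sprites and looping sparks add detail.
# Time-of-day and front-end values adjust color, alpha and motion.
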NiagaraSystem /BRCosmetics/Animation/Game/MainPlayer/Emotes/FlowingFlames/VFX/Refs/NS_Emote_FlowingFlames_Ref
dependencies {
 usedDataInterfaces [
  &NiagaraDataInterfaceArrayFloat3
  &NiagaraDataInterfaceArrayInt32
  &NiagaraDataInterfaceArrayFloat
  &NiagaraDataInterfaceCurve
  &NiagaraDataInterfaceSpriteRendererInfo
  &NiagaraDataInterfaceColorCurve
  &NiagaraDataInterfacePlatformSet
 ]
 usedDynamicInputs [
  /Niagara/DynamicInputs/Lerp/SmoothLerpOverTime_Float
  &RemapRange
  &SetBoolByFloatComparison
  &Cosine
  &Add_Float
  &Divide_Float
 ]
 usedModules [
  &self:SystemSpawnScript.SystemScriptSource.SystemScriptGraph.NiagaraNodeAssignment_0.SetVariables_#1
  &self:SystemSpawnScript.SystemScriptSource.SystemScriptGraph.NiagaraNodeAssignment_3.SetVariables_#2
  &self:SystemSpawnScript.SystemScriptSource.SystemScriptGraph.NiagaraNodeAssignment_5.SetVariables_#3
  &SystemState
  &Timeline
  &self:SystemSpawnScript.SystemScriptSource.SystemScriptGraph.NiagaraNodeAssignment_8.SetVariables_#4
  &self:SystemSpawnScript.SystemScriptSource.SystemScriptGraph.NiagaraNodeAssignment_17.SetVariables_#5
  &self:SystemSpawnScript.SystemScriptSource.SystemScriptGraph.NiagaraNodeAssignment_1.SetVariables_#6
 ]
}
emitters {
 Droplets {
  emitterUpdateScript [
   &EmitterState()
   &SpawnRate(SpawnRate=200, Use Spawn Probability=false)
  ]
  particleSpawnScript [
   &InitializeParticle(
    Color Maximum=<rgba 0.2 0.8 1 1>
    Color Minimum=<rgba 0 0.7333 0.8 1>
    Color Mode=::NewEnumerator2
    Lifetime Max=0.33
    Lifetime Min=0.1
    Lifetime Mode=::NewEnumerator1
    Sprite Rotation Mode=::NewEnumerator2
    Sprite Size Mode=::NewEnumerator3
    Uniform Sprite Size=&RandomRangeFloat(Minimum=1, Maximum=2)
    UsePositionOffset=false
    Write Material Random=true
   )
   &SubUVAnimation(
    "Sprite Renderer "=DataInterface() {propertyValues {SpriteRenderer=&self:Droplets_1.NiagaraSpriteRendererProperties_0}}
    SubUV Animation Mode:ENiagara_SubUVLookupModeV2=::NewEnumerator2
    UseEndFrame=false
    UseStartFrame=false
   )
   &SocketLocation(
    Interpolate Socket Sampling=true
    SamplePosOffset=true
    Socket Index=1
    Socket Reader=$System.Socket Reader
    Socket Sampling Mode=::NewEnumerator3
    VelocitySpeedLimit=false
   )
   &ShapeLocation(MaskSpawnGroup=false, Override Seed=false, Sphere Radius=5, Sphere Surface Distribution=0.5238, UseOwnerScale=true)
   /SetVariables(Particles.Velocity=&Multiply_VectorByFloat(Vector=$Output.SocketLocation.SampledVelocity, Float=&RandomRangeFloat(Minimum=0.05, Maximum=0.1))) {isSetParameters=true}
   &AddVelocity(
    Module.Use Falloff Distance=false
    Override Seed=false
    Velocity Mode=::NewEnumerator1
    Velocity Origin=$Output.SocketLocation.SampledPosition
    Velocity Speed=&SelectFloatFromArray(Direct Array Index=$Particles.PartitionParticles.Partition, Float Selection Array=DataInterface() {propertyValues {FloatData [20,150]}})
   ) {enabled=false}
   &AddVelocity(Velocity=$Output.SocketLocation.SampledVelocity) {enabled=false, module=AddVelocity001}
  ]
  particleUpdateScript [
   &ParticleState()
   &ScaleSpriteSize(Uniform Curve Sprite Scale=DataInterface() {propertyValues {Curve {keys [
        {interpMode=RCIM_Cubic, tangentMode=RCTM_User, time=0.33, value=1}
        {arriveTangent=-2.8079, interpMode=RCIM_Cubic, leaveTangent=-2.8079, tangentMode=RCTM_User, time=1, value=0}
       ]}}})
   &Drag(Drag=&RandomRangeFloat(Minimum=6, Maximum=8), Use Linear Drag=true, Use Rotational Drag=true)
   &GravityForce(Gravity=&Multiply_VectorByFloat(
     Vector=&RandomRangeVector(Minimum=<0 0 -250>, Maximum=<0 0 -350>)
     Float=&FN_FrontEndFloatModifier(Gameplay=1, Front End=0.66)
    ))
   &SolveForcesAndVelocity(Clamp Velocity=false, Limit Acceleration=false)
   &ScaleSpriteSizeBySpeed(Max Scale Factor=<0.8 2.5>, Velocity Threshold=100)
   &ScaleColor(
    Scale Alpha=&FloatFromCurve(FloatCurve=DataInterface() {propertyValues {Curve {
        defaultValue=1
        keys [
         {interpMode=RCIM_Cubic, tangentMode=RCTM_User, time=0, value=1}
         {arriveTangent=-2.0915, interpMode=RCIM_Cubic, leaveTangent=-2.0915, tangentMode=RCTM_User, time=1, value=0}
        ]
       }}})
    Scale RGB=&MakeVectorFromLinearColorRGB(Color=&ColorFromCurve(ColorCurve=DataInterface() {propertyValues {
        AlphaCurve {keys [<at 0 1 RCIM_Cubic>]}
        BlueCurve {keys [<at 0.0006 1>]}
        GreenCurve {keys [<at 0.0006 1>]}
        RedCurve {keys [<at 0.0006 1>]}
       }}))
    ScaleA=true
    ScaleRGB=true
   )
   &ScaleColor(
    Color Value To Scale=$Particles.Color
    Scale RGB=&MakeVectorFromLinearColorRGB(Color=<rgba 1 1 1 1>)
    ScaleA=false
    ScaleRGB=true
   ) {module=ScaleColor001}
  ]
  rendererClasses [&NiagaraSpriteRendererProperties]
  renderers [{rendererClass=&NiagaraSpriteRendererProperties}]
 }
 Glow_and_Liquid {
  emitterUpdateScript [
   &EmitterState()
   &FN_SpawnAnimTrail()
  ]
  particleSpawnScript [
   &FN_InitializeAnimTrailBoneIndecies(First Bone or Socket Index=1)
   &FN_InitializeAnimTrail(AnimTrailWidth=25, Color=<rgba 0 0.3304 0.6609 1>, Lifetime=0.22)
  ]
  particleUpdateScript [
   &FN_UpdateAnimTrail()
   &ScaleColor(
    Scale Alpha=&FloatFromCurve(
     FloatCurve=DataInterface() {propertyValues {Curve {keys [
         {arriveTangent=9.3297, interpMode=RCIM_Cubic, leaveTangent=9.3297, tangentMode=RCTM_User, time=0, value=0.5}
         {arriveTangent=0.0028, interpMode=RCIM_Cubic, leaveTangent=0.0028, tangentMode=RCTM_User, time=0.1, value=1}
         {interpMode=RCIM_Cubic, tangentMode=RCTM_Break, time=1, value=0}
        ]}}}
     Scale Curve=$System.SwingGeneralFade
    )
    Scale RGB=&VectorFromFloat(Value=&FloatFromCurve(FloatCurve=DataInterface() {propertyValues {Curve {keys [
          {arriveTangent=-0.005, interpMode=RCIM_Cubic, leaveTangent=-0.005, tangentMode=RCTM_User, time=0, value=1}
          {arriveTangent=-0.7321, tangentMode=RCTM_User, time=1, value=0.6}
         ]}}}))
    ScaleA=true
    ScaleRGB=true
   )
   /SetVariables(Particles.GlowColor=&MultiplyLinearColor(
     Linear Color=<rgba 0.0783 0.5391 1 1>
     Scale Factor=&MakeVector4FromVectorAndFloat(
      "Vector (RGB)"=&MakeVectorFromLinearColorRGB(Color=$System.TODColor)
      "Float (Alpha)"=&FloatFromCurve(
       FloatCurve=DataInterface() {propertyValues {Curve {
          defaultValue=1
          keys [
           {interpMode=RCIM_Cubic, tangentMode=RCTM_User, time=0, value=1}
           {arriveTangent=-1.0589, interpMode=RCIM_Cubic, leaveTangent=-1.0589, tangentMode=RCTM_User, time=1, value=0}
          ]
         }}}
       Scale Curve=$System.SwingGeneralFade
      )
     )
    )) {isSetParameters=true}
   &ScaleRibbonWidth(Ribbon Width Scale=&FloatFromCurve(FloatCurve=DataInterface() {propertyValues {Curve {
        defaultValue=1
        keys [{interpMode=RCIM_Cubic, tangentMode=RCTM_User, time=0, value=1}, {interpMode=RCIM_Cubic, tangentMode=RCTM_User, time=1, value=0}]
       }}}))
   /SetVariables(Particles.GlowRibbonWidth:NiagaraFloat=&Multiply_Float(A=$Particles.RibbonWidth, B=1.25)) {isSetParameters=true}
   &DynamicMaterialParameters(
    Index 0 Param 1=1
    Index 0 Param 2=1
    Module.Index0Float0Write=true
    Module.Index0Float1Write=true
    Module.Index0Float2Write=true
    Module.Index0Float3Write=true
   )
   &SolveForcesAndVelocity(Clamp Velocity=false, Limit Acceleration=false)
  ]
  rendererClasses [&NiagaraRibbonRendererProperties]
  renderers [
   {rendererClass=&NiagaraRibbonRendererProperties}
   {rendererClass=&NiagaraRibbonRendererProperties, rendererIndex=1}
  ]
 }
 NE_Fire_01 {
  emitterUpdateScript [
   &EmitterState()
   &SpawnBurst_Instantaneous(Spawn Count=12, Use Spawn Probability=false, UseLoopCountLimit=false)
  ]
  particleSpawnScript [
   &PartitionParticles(Number Of Partitions=12)
   &InitializeParticle(
    Color Maximum=<rgba 1 1 1 0.2>
    Color Minimum=<rgba 1 1 1 0.1>
    Color Mode=::NewEnumerator2
    Lifetime=2
    Mesh Scale=&Multiply_Vector(
     A=&SelectVectorFromArray(
      Direct Array Index=$Particles.PartitionParticles.Partition
      Vector Selection Array=DataInterface() {propertyValues {FloatData [
         <0.8 0.7 15>
         <0.8 0.8 5>
         <0.5 0.8 20>
         <0.5 0.8 -40>
         <0.7 0.8 30>
         <0.6 1.2 -60>
         <0.6 1 30>
         <0.6 1 -40>
         <0.6 0.7 30>
         <0.6 0.75 -30>
         <0.6 0.8 1>
         <0.6 0.8 1>
        ]}}
     )
     B=<0.95 0.95 1>
    )
    Mesh Scale Mode=::NewEnumerator4
    UsePositionOffset=false
    Write Material Random=true
   )
   &InitialMeshOrientation(Mesh Orientation Mode=::NewEnumerator2, Rotation=<0 0 0.25>, Use Rotation Vector=true)
  ]
  particleUpdateScript [
   &ParticleState()
   &SkeletalMeshLocation(
    Bone / Socket Index=&SelectIntFromArray(
     Direct Array Index=$Particles.PartitionParticles.Partition
     Int Selection Array=DataInterface() {propertyValues {IntData [20,21,1,2,1,2,5,6,5,6,20,21]}}
    )
    Bone Sampling Mode=::NewEnumerator4
    MaskSpawnGroup=false
    Sampled Position Offset=&SelectVectorFromArray(
     Direct Array Index=$Particles.PartitionParticles.Partition
     Vector Selection Array=DataInterface() {propertyValues {FloatData [<15 0 0>, <-12 0 -2>, <5 0 3>, <-5 0 0>, <17 0 2>, <-15 0 0>, <6 0 2>, <-6 0 -2>, <18 0 0>, <-18 0 0>, <5 0 5>, <-5 0 -4>]}}
    )
    Skeletal Mesh=$System.Skeletal Mesh Reader
   )
   &DynamicMaterialParameters(
    Index 0 Param 1=&RandomRangeFloat(Minimum=0.4, Maximum=0.8)
    Index 0 Param 2=&RandomRangeFloat(Minimum=0.4, Maximum=0.8)
    Index 0 Param 4=&RandomRangeFloat(Minimum=1, Maximum=1.3)
    Index 2 Param 1=&RandomRangeFloat(Minimum=0.8, Maximum=1)
    Index 2 Param 2=&Multiply_Float(
     A=&SelectFloatFromArray(
      Direct Array Index=$Particles.PartitionParticles.Partition
      Float Selection Array=DataInterface() {propertyValues {FloatData [1.3,1.5,0.5,0.8,1,0.8,0.85,0.8,0.75,0.6,1.5,1.5]}}
     )
     B=0.65
    )
    Module.Index0Float0Write=true
    Module.Index0Float1Write=true
    Module.Index0Float2Write=true
    Module.Index0Float3Write=true
    Module.Index2Float0Write=true
    Module.Index2Float1Write=true
    Module.Index2Float2Write=true
    Module.Index2Float3Write=true
    Write Parameter Index 2=::NewEnumerator1
   )
   &InterpolateOverTime(Rate Of Change=5, Target Vector Value:NiagaraPosition=$Particles.Position, Variable Type=::NewEnumerator2)
   /SetVariables(Particles.Interpolate Dynamic=&MakeFloat4FromVectorAndFloat("Vector (RGB)"=$Particles.InterpolateOverTime.InterpolateOverTime_V3_.Moving Average, "Float (Alpha)"=0.4)) {isSetParameters=true}
   &ScaleColor(
    Scale Alpha=&MakeCustomFloatFromBool(
     Bool=$User.bIsFrontEndPreview
     True Float=0.3
     False Float=&FloatFromCurve(FloatCurve=DataInterface() {propertyValues {Curve {keys [<at 0 1>,<at 0.5 1>,<at 1 0>]}}})
    )
    Scale RGB=&MakeVectorFromLinearColorRGB(Color=<rgba 1 1 1 1>)
    ScaleA=true
    ScaleRGB=true
   )
  ]
  rendererClasses [&NiagaraMeshRendererProperties]
  renderers [{rendererClass=&NiagaraMeshRendererProperties, rendererIndex=0}]
 }
 NE_Smoke {
  emitterUpdateScript [
   &EmitterState()
   &SpawnRate(SpawnRate=15, Use Spawn Probability=false)
   &SpawnRate(SpawnGroup=1, SpawnRate=50, Use Spawn Probability=false) {module=SpawnRate001}
  ]
  particleSpawnScript [
   /SetVariables(Particles.VisibilityTag=&SpawnGroup()) {isSetParameters=true}
   &InitializeParticle(
    Color=&MergeHelper_ColorSwitch(0=<rgba 10 0.981 0 0.55>, 1=<rgba 1 1 1 0.05>, 2=?, 3=?, Index=$Particles.VisibilityTag, Num=::NewEnumerator0)
    Lifetime Max=0.6
    Lifetime Min=0.3
    Lifetime Mode=::NewEnumerator1
    Position Offset=<0 -5 0.5>
    Sprite Rotation Mode=::NewEnumerator2
    Sprite Size Mode=::NewEnumerator3
    Uniform Sprite Size=&MergeHelper_FloatSwitch(
     0=&RandomRangeFloat(Minimum=25, Maximum=40)
     1=&RandomRangeFloat(Minimum=15, Maximum=30)
     2=?
     3=?
     Index=$Particles.VisibilityTag
     Num=::NewEnumerator0
    )
    UsePositionOffset=false
    Write Material Random=true
   )
   &SocketLocation(
    OverrideRandomSeed=false
    SamplePosOffset=false
    Sampled Position Offset=<0 -3 0>
    Socket Reader=$System.Socket Reader
    Socket Sampling Mode=::NewEnumerator1
    VelocitySpeedLimit=false
   )
   /SetVariables(Particles.FlamesDY:Vector4f=&MakeVector4(X=0, Y=&RandomRangeFloat(Minimum=0, Maximum=4), Z=0, W=2.5)) {isSetParameters=true}
  ]
  particleUpdateScript [
   &ParticleState()
   /SetVariables(Particles.FlamesDY:Vector4f=&MakeVector4(
     X=&FloatFromCurve(FloatCurve=DataInterface() {propertyValues {Curve {keys [<at 0 1>]}}})
     Y=0.5
     Z=&FloatFromCurve(FloatCurve=DataInterface() {propertyValues {Curve {keys [<at 0 0.5 RCIM_Cubic>, {arriveTangent=1.5, interpMode=RCIM_Cubic, leaveTangent=1.5, time=0.5235, value=0.5}, <at 1 2>]}}})
     W=0
    )) {isSetParameters=true}
   &ScaleSpriteSize(
    Scale Sprite Size Mode=::NewEnumerator0
    Uniform Scale Factor=&MergeHelper_FloatSwitch(
     0=&FloatFromCurve(FloatCurve=DataInterface() {propertyValues {Curve {keys [<at 0 1>,<at 1 0>]}}})
     1=1
     2=?
     3=?
     Index=$Particles.VisibilityTag
     Num=::NewEnumerator0
    )
   )
   &SubUVAnimation(
    End Frame Range Override=64
    "Sprite Renderer "=DataInterface() {propertyValues {SpriteRenderer=&self:NE_Smoke_1.NiagaraSpriteRendererProperties_0}}
    SubUV Animation Mode:ENiagara_SubUVLookupModeV2=::NewEnumerator3
    UseEndFrame=false
    UsePlayRateScale=false
    UseStartFrame=false
   )
   &ScaleColor(
    Scale Alpha=&MergeHelper_FloatSwitch(
     0=&FloatFromCurve(FloatCurve=DataInterface() {propertyValues {Curve {keys [
          <at 0 0 RCIM_Cubic>
          {arriveTangent=-0.0116, interpMode=RCIM_Cubic, leaveTangent=-0.0116, tangentMode=RCTM_User, time=0.4073, value=1}
          {arriveTangent=-2.0743, interpMode=RCIM_Cubic, leaveTangent=-2.0743, tangentMode=RCTM_User, time=0.6746, value=0.6742}
          {interpMode=RCIM_Cubic, leaveTangent=-2.371, time=1, value=0}
         ]}}})
     1=&FloatFromCurve(
      FloatCurve=DataInterface() {propertyValues {Curve {keys [
          {interpMode=RCIM_Cubic, tangentMode=RCTM_User, time=0, value=1}
          {arriveTangent=0.0033, interpMode=RCIM_Cubic, leaveTangent=-5.4441, tangentMode=RCTM_Break, time=0.5, value=1}
          {interpMode=RCIM_Cubic, tangentMode=RCTM_User, time=1, value=0}
         ]}}}
      Scale Curve=&Lerp_Float(
       A=&FNTimeOfDayScalar(Night Scale=1, Morning Scale=1.5, Day Scale=2, Evening Scale=1.5)
       B=0
       Alpha=&MakeCustomFloatFromBool(Bool=$User.bisSwinging, True Float=1, False Float=0)
      )
     )
     2=?
     3=?
     Index=$Particles.VisibilityTag
     Num=::NewEnumerator0
    )
    Scale RGB=&MergeHelper_VectorSwitch(
     0=&VectorFromFloat(Value=&FloatFromCurve(FloatCurve=DataInterface() {propertyValues {Curve {keys [
           <at 0 1>
           {interpMode=RCIM_Cubic, tangentMode=RCTM_User, time=0.5, value=1}
           {arriveTangent=-3.29, interpMode=RCIM_Cubic, leaveTangent=-3.29, tangentMode=RCTM_User, time=1, value=0}
          ]}}}))
     1=<1 1 1>
     2=?
     3=?
     Index=$Particles.VisibilityTag
     Num=::NewEnumerator0
    )
    ScaleA=true
    ScaleRGB=true
   )
   &Drag(Drag=&RandomRangeFloat(Minimum=3, Maximum=6), Use Linear Drag=true, Use Rotational Drag=true)
   &CurlNoiseForce(Bias Noise=false, Noise Strength=750, Pan Noise=false)
   &SolveForcesAndVelocity(Clamp Velocity=false, Limit Acceleration=false)
  ]
  rendererClasses [&NiagaraSpriteRendererProperties]
  renderers [
   {rendererClass=&NiagaraSpriteRendererProperties}
   {rendererClass=&NiagaraSpriteRendererProperties, rendererIndex=1}
  ]
 }
 NE_Sparks {
  emitterUpdateScript [
   &EmitterState(Life Cycle Mode=::NewEnumerator1, Loop Behavior=::NewEnumerator1, Loop Delay=0.07, Loop Duration=0.15, UseLoopDelay=true)
   &SpawnRate(SpawnRate=120, Use Spawn Probability=false)
  ]
  particleSpawnScript [
   &InitializeParticle(
    Color=<rgba 2 2 2 1>
    Lifetime Max=0.5
    Lifetime Min=0.2
    Lifetime Mode=::NewEnumerator1
    Sprite Size Max=<2 6>
    Sprite Size Min=<1 2>
    Sprite Size Mode=::NewEnumerator2
    UsePositionOffset=false
    Write Material Random=true
   )
   &SkeletalMeshLocation(
    Bone / Socket Index=1
    MaskSpawnGroup=false
    Mesh Sampling Type=::NewEnumerator5
    Sampled Position Offset=<0 0 -10>
    Skeletal Mesh=$User.SkeletalMesh
    Socket Sampling Mode=::NewEnumerator5
   )
   &ShapeLocation(
    MaskSpawnGroup=false
    Offset Mode=::NewEnumerator0
    Override Seed=false
    Rotation Axis=<0 1 0>
    Rotation Mode=::NewEnumerator1
    Sphere Radius=15
    UseOwnerScale=true
   )
   &AddVelocity(Velocity=&RandomRangeVector(Minimum=<0 0 50>, Maximum=<0 0 80>))
   &InheritVelocity(
    Inherited Velocity Speed Limit:NiagaraFloat=&RandomRangeFloat(Minimum=50, Maximum=100)
    Speed Limit=true
    UseSpeedThreshold=false
   )
  ]
  particleUpdateScript [
   &ParticleState()
   &ScaleSpriteSize(
    Uniform Curve Scale=&FloatFromCurve(FloatCurve=DataInterface() {propertyValues {Curve {keys [<at 0 0>,<at 0.17 0>,<at 0.2 1.2>]}}}, CurveIndex=$Emitter.Age)
    Uniform Curve Sprite Scale=DataInterface() {propertyValues {Curve {keys [<at 0 0 RCIM_Cubic>,<at 0.2 1 RCIM_Cubic>,<at 1 0 RCIM_Cubic>]}}}
   )
   /SetVariables(Particles.Glow03_Size=&Multiply_Vector2D_ByFloat(
     Float=&FloatFromCurve(
      FloatCurve=DataInterface() {propertyValues {Curve {keys [
          <at 0 0 RCIM_Cubic>
          <at 0.2 1.3 RCIM_Cubic>
          {arriveTangent=-0.1423, interpMode=RCIM_Cubic, leaveTangent=-0.1423, tangentMode=RCTM_User, time=1, value=0}
         ]}}}
      Scale Curve=15
     )
     Vector2D=<0.7 1.2>
    )) {isSetParameters=true}
   &ScaleColor(
    Scale Alpha=&Multiply_Float(
     A=1
     B=&FloatFromCurve(
      FloatCurve=DataInterface() {propertyValues {Curve {keys [
          {interpMode=RCIM_Cubic, tangentMode=RCTM_User, time=0.1, value=0}
          {arriveTangent=-1.3622, interpMode=RCIM_Cubic, leaveTangent=-1.3622, tangentMode=RCTM_User, time=0.2, value=1}
          {tangentMode=RCTM_User, time=0.8, value=0}
         ]}}}
      Scale Curve=&FloatFromCurve(FloatCurve=DataInterface() {propertyValues {Curve {keys [<at 0 0>,<at 0.15 0>,<at 0.25 1>]}}}, CurveIndex=$Emitter.Age)
     )
    )
    Scale RGB=&MakeVectorFromLinearColorRGB(Color=&ColorFromCurve(ColorCurve=DataInterface() {propertyValues {
        AlphaCurve {keys []}
        BlueCurve {keys [<at 0 0 RCIM_Cubic>]}
        GreenCurve {keys [<at 0 0.3354 RCIM_Cubic>]}
        RedCurve {keys [<at 0 1.2 RCIM_Cubic>]}
       }}))
    ScaleA=true
    ScaleRGB=true
   )
   /SetVariables(Particles.Glow03_Color=&MakeLinearColorFromVectorAndFloat(
     "Vector (RGB)"=&MakeVectorFromLinearColorRGB(Color=&MultiplyLinearColor(
       Linear Color=<rgba 1 0.15 0.02 0.2>
       Scale Factor=&MakeVector4_FromFloat(Float=&Multiply_Float(A=1, B=&FNTimeOfDayScalar(Night Scale=0.5, Morning Scale=1, Evening Scale=1)))
      ))
     "Float (Alpha)"=&Multiply_Float(
      A=0.2
      B=&FloatFromCurve(FloatCurve=DataInterface() {propertyValues {Curve {keys [<at 0 0>,<at 0.15 0>,<at 0.25 1>]}}}, CurveIndex=$Emitter.Age)
     )
    )) {isSetParameters=true}
   &CurlNoiseForce(Bias Noise=false, Noise Strength=500, Pan Noise=false)
   &Drag(
    Drag=&FloatFromCurve(
     FloatCurve=DataInterface() {propertyValues {Curve {keys [<at 0 1 RCIM_Cubic>, <at 0.5 0.6 RCIM_Cubic>, {arriveTangent=0.6277, tangentMode=RCTM_User, time=1, value=1}]}}}
     Scale Curve=4
    )
    Use Linear Drag=true
    Use Rotational Drag=true
   )
   &SolveForcesAndVelocity(Clamp Velocity=false, Limit Acceleration=false)
   &ScaleSpriteSizeBySpeed(Max Scale Factor=<1 2>)
  ]
  rendererClasses [&NiagaraSpriteRendererProperties]
  renderers [
   {rendererClass=&NiagaraSpriteRendererProperties}
   {rendererClass=&NiagaraSpriteRendererProperties, rendererIndex=1}
  ]
 }
 SplashSprites {
  emitterUpdateScript [
   &EmitterState()
   &SpawnBurst_Instantaneous(
    Spawn Count=&RandomRangeInt(Minimum=3, Maximum=4)
    Spawn Time=0.02
    Use Spawn Probability=false
    UseLoopCountLimit=false
   )
  ]
  particleSpawnScript [
   &InitializeParticle(
    Color=&RandomRangeLinearColor(Minimum=<rgba 0.1042 0.7014 1 1>, Maximum=<rgba 0 1.5 1.25 1>)
    Lifetime Max=0.7
    Lifetime Min=0.5
    Lifetime Mode=::NewEnumerator1
    Position Mode=::NewEnumerator0
    Sprite Rotation Mode=::NewEnumerator2
    Sprite Size Mode=::NewEnumerator1
    Sprite UV Mode=::NewEnumerator3
    Uniform Sprite Size Max=120
    Uniform Sprite Size Min=80
    Write Material Random=true
   )
   /SetVariables(Particles.SubImageIndex=&RandomRangeFloat(Minimum=0, Maximum=3)) {isSetParameters=true}
   &ShapeLocation(MaskSpawnGroup=false, Override Seed=false, Sphere Radius=2, Sphere Surface Distribution=0.2, UseOwnerScale=true)
   &AddVelocity(
    Module.Use Falloff Distance=false
    Origin Offset=<-5 0 -2>
    Override Seed=false
    Velocity Mode=::NewEnumerator1
    Velocity Speed=&Multiply_Float(
     A=&Multiply_Float(
      A=&RandomRangeFloat(Minimum=100, Maximum=200)
      B=&FN_PickFloatDependOnPlatform(TargetPlatform=DataInterface() {propertyValues {Platforms {qualityLevelMask=-2}}}, TargetPlatformActive=1, TargetPlatformNoActive=0.7)
     )
     B=$System.FrontendScale
    )
   )
  ]
  particleUpdateScript [
   &ParticleState()
   &SpriteRotationRate(Rotation Rate=&FloatFromCurve(
     FloatCurve=DataInterface() {propertyValues {Curve {
        defaultValue=1
        keys [{interpMode=RCIM_Cubic, tangentMode=RCTM_User, time=0, value=1}, {interpMode=RCIM_Cubic, tangentMode=RCTM_User, time=1, value=0}]
       }}}
     Scale Curve=&Multiply_FloatByInt(
      Float=&RandomRangeFloat(Minimum=60, Maximum=100)
      Integer=&SelectIntFromArray(
       Array Sampling Mode=::NewEnumerator1
       Direct Array Index=?
       Int Selection Array=DataInterface() {propertyValues {IntData [-1,1]}}
       Random Seed=0
      )
     )
    ))
   &GravityForce(Gravity=&Multiply_VectorByFloat(
     Vector=&Multiply_VectorByFloat(Vector=<0 0 -600>, Float=&RandomRangeFloat(Minimum=0.7, Maximum=1))
     Float=$System.FrontendScale
    ))
   &Drag(Drag=&RandomRangeFloat(Minimum=2, Maximum=3), Use Linear Drag=true, Use Rotational Drag=true)
   &SolveForcesAndVelocity(Clamp Velocity=false, Limit Acceleration=false)
   &ScaleSpriteSize(
    Uniform Curve Scale=$System.FrontendScale
    Uniform Curve Sprite Scale=DataInterface() {propertyValues {Curve {keys [
        {arriveTangent=1.7973, interpMode=RCIM_Cubic, leaveTangent=1.7973, tangentMode=RCTM_User, time=0, value=0.1}
        {interpMode=RCIM_Cubic, tangentMode=RCTM_User, time=1, value=1}
       ]}}}
   )
   &ScaleColor(
    Scale Alpha=&FloatFromCurve(FloatCurve=DataInterface() {propertyValues {Curve {
        defaultValue=1
        keys [{interpMode=RCIM_Cubic, tangentMode=RCTM_User, time=0, value=1}, {interpMode=RCIM_Cubic, tangentMode=RCTM_User, time=1, value=0}]
       }}})
    Scale RGB=&MakeVectorFromLinearColorRGB(Color=&ColorFromCurve(ColorCurve=DataInterface() {propertyValues {
        AlphaCurve {keys [<at 0 1 RCIM_Cubic>]}
        BlueCurve {keys [<at 0 1>,<at 0.8321 0.1>]}
        GreenCurve {keys [<at 0 1>,<at 0.8321 0.0825>]}
        RedCurve {keys [<at 0 1>,<at 0.8321 0.08>]}
       }}))
    ScaleA=true
    ScaleRGB=true
   )
   &ScaleColor(
    Scale RGB=&FN_FrontEndVector3Modifier(
     Gameplay=&MakeVectorFromLinearColorRGB(Color=&FNTimeOfDayColor(Night Color=<rgba 0.32 0.34 0.4 1>, Morning Color=<rgba 0.6 0.468 0.42 1>, Day Color=<rgba 1 1 1 1>, Evening Color=<rgba 0.8 0.52 0.4 1>))
     Front End=&MakeVectorFromLinearColorRGB(Color=<rgba 0 1.3333 2 1>)
    )
    ScaleA=true
    ScaleRGB=true
   ) {enabled=false, module=ScaleColor001}
   &DynamicMaterialParameters(
    Index 0 Param 1=&RandomRangeFloat(Minimum=0, Maximum=1)
    Index 0 Param 2=&FloatFromCurve(
     FloatCurve=DataInterface() {propertyValues {Curve {
        defaultValue=1
        keys [
         {arriveTangent=-2.3912, interpMode=RCIM_Cubic, leaveTangent=-2.3912, tangentMode=RCTM_User, time=0, value=1}
         {interpMode=RCIM_Cubic, tangentMode=RCTM_User, time=0.7, value=0}
        ]
       }}}
     Scale Curve=&RandomRangeFloat(Minimum=0.5, Maximum=1)
    )
    Index 0 Param 3=&Lerp_Float(
     A=0
     B=3
     Alpha=&FloatFromCurve(FloatCurve=DataInterface() {propertyValues {Curve {
         defaultValue=0
         keys [
          {interpMode=RCIM_Cubic, tangentMode=RCTM_User, time=0, value=0}
          {arriveTangent=3.1748, interpMode=RCIM_Cubic, leaveTangent=3.1748, tangentMode=RCTM_User, time=1, value=1}
         ]
        }}})
    )
    Index 0 Param 4=&FloatFromCurve(FloatCurve=DataInterface() {propertyValues {Curve {
        defaultValue=0
        keys [
         {arriveTangent=2.4825, interpMode=RCIM_Cubic, leaveTangent=2.4825, tangentMode=RCTM_User, time=0, value=0}
         {interpMode=RCIM_Cubic, tangentMode=RCTM_User, time=1, value=1}
        ]
       }}})
    Module.Index0Param0Write=true
    Module.Index0Param1Write=true
    Module.Index0Param2Write=true
    Module.Index0Param3Write=true
   )
  ]
  rendererClasses [&NiagaraSpriteRendererProperties]
  renderers [{rendererClass=&NiagaraSpriteRendererProperties}]
 }
 Toon_Water {
  emitterUpdateScript [
   &EmitterState()
   &FN_SpawnAnimTrail()
  ]
  particleSpawnScript [
   &FN_InitializeAnimTrailBoneIndecies(First Bone or Socket Index=1)
   &FN_InitializeAnimTrail(AnimTrailWidth=25, Lifetime=0.22)
  ]
  particleUpdateScript [
   &FN_UpdateAnimTrail()
   &ScaleColor(
    Scale Alpha=&FloatFromCurve(
     FloatCurve=DataInterface() {propertyValues {Curve {defaultValue=1, keys [{interpMode=RCIM_Cubic, tangentMode=RCTM_User, time=0, value=1}]}}}
     Scale Curve=$System.SwingGeneralFade
    )
    Scale RGB=&MakeVectorFromLinearColorRGB(Color=$System.TODColor)
    ScaleA=true
    ScaleRGB=true
   )
   /SetVariables(Particles.GlowColor=&MultiplyLinearColor(
     Linear Color=<rgba 0.3391 0.9196 1.5 1>
     Scale Factor=&MakeVector4FromVectorAndFloat(
      "Vector (RGB)"=&MakeVectorFromLinearColorRGB(Color=$System.TODColor)
      "Float (Alpha)"=&FloatFromCurve(
       FloatCurve=DataInterface() {propertyValues {Curve {
          defaultValue=1
          keys [
           {interpMode=RCIM_Cubic, tangentMode=RCTM_User, time=0, value=1}
           {arriveTangent=-1.0589, interpMode=RCIM_Cubic, leaveTangent=-1.0589, tangentMode=RCTM_User, time=1, value=0}
          ]
         }}}
       Scale Curve=$System.SwingGeneralFade
      )
     )
    )) {isSetParameters=true}
   &ScaleRibbonWidth(Ribbon Width Scale=&Multiply_Float(
     A=&FloatFromCurve(FloatCurve=DataInterface() {propertyValues {Curve {
         defaultValue=0
         keys [
          {interpMode=RCIM_Cubic, tangentMode=RCTM_User, time=0, value=1}
          {arriveTangent=-1.5474, interpMode=RCIM_Cubic, leaveTangent=-1.5474, tangentMode=RCTM_User, time=1, value=0}
         ]
        }}})
     B=1
    ))
  ]
  rendererClasses [&NiagaraRibbonRendererProperties]
  renderers [
   {rendererClass=&NiagaraRibbonRendererProperties}
   {rendererClass=&NiagaraRibbonRendererProperties, rendererIndex=1}
  ]
 }
}
summary {
 systemName=NS_Emote_FlowingFlames_Ref
 emitters [
  {emitterName=NE_Fire_01, rendererClasses [&NiagaraMeshRendererProperties]}
  {emitterName=Toon_Water, rendererClasses [&NiagaraRibbonRendererProperties]}
  {emitterName=Glow_and_Liquid, rendererClasses [&NiagaraRibbonRendererProperties]}
  {emitterName=Droplets, rendererClasses [&NiagaraSpriteRendererProperties]}
  {emitterName=NE_Smoke, rendererClasses [&NiagaraSpriteRendererProperties]}
  {emitterName=NE_Sparks, rendererClasses [&NiagaraSpriteRendererProperties]}
  {emitterName=SplashSprites, rendererClasses [&NiagaraSpriteRendererProperties]}
 ]
 userVariables [
  User.bIsADS {type=NiagaraBool}
  User.bIsFrontEnd {type=NiagaraBool}
  User.bIsFrontEndPreview {type=NiagaraBool}
  User.bisSwinging {type=NiagaraBool}
  User.IsSkydiving {type=NiagaraBool}
  User.RibbonIDOffset {type=NiagaraInt32}
  User.Selector {type=NiagaraInt32}
  User.SkeletalMesh {
   type=NiagaraDataInterfaceSkeletalMesh
   defaultValue {
    struct=NiagaraExt_VariableValue_DataInterface
    value {dataInterface=&self:NiagaraDataInterfaceSkeletalMesh_0}
   }
  }
 ]
}
